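# Niagara system for the Trace vehicle boost cosmetic, drawing flares and lightning trails at the car's boost sockets.
# FlareGroup merges five spawn groups of sprites, Lightning adds short-lived spinning meshes and sprites, and RibbonLightningGroup streams multi-layer ribbons that inherit fading source motion.
# Emitters spread particles across filtered skeletal mesh sockets by port count, with painted-variant color overrides throughout.
# An older ribbon emitter, RibbonLightninGroup, is disabled.
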NiagaraSystem /VehicleCosmetics/Boosts/Boost_Trace/FX/NS_Boost_Trace_MP
dependencies {
 usedDataInterfaces [&NiagaraDataInterfaceCurve, &NiagaraDataInterfaceSkeletalMesh]
 usedDynamicInputs [&Multiply_FloatByInt]
 usedModules [
  &self:SystemSpawnScript.SystemScriptSource.SystemScriptGraph.NiagaraNodeAssignment_0.SetVariables_#1
  &SystemState
 ]
}
emitters {
 FlareGroup {
  emitterUpdateScript [
   &EmitterState(Inactive Response=::NewEnumerator1, Life Cycle Mode=::NewEnumerator1, Loop Duration=1, UseLoopDelay=false)
   &VC_FilteredSocketCount()
   &SpawnRate(SpawnRate=10, Use Spawn Probability=false)
   &SpawnRate(SpawnGroup=1, SpawnRate=10, Use Spawn Probability=false) {module=SpawnRate001}
   &SpawnRate(SpawnGroup=2, SpawnRate=10, Use Spawn Probability=false) {module=SpawnRate002}
   &SpawnRate(SpawnGroup=3, SpawnRate=10, Use Spawn Probability=false) {module=SpawnRate003}
   &SpawnRate(SpawnGroup=4, SpawnRate=30, Use Spawn Probability=false) {module=SpawnRate004}
  ]
  particleSpawnScript [
   &VC_PaintedVariantsOverride(DefaultColor=<rgba 0.3944 0.3 2 1>) {enabled=false}
   &VC_PaintedVariantsOverride(
    Black=<rgba 0.05 0.05 0.05 1>
    BurntSienna=<rgba 0.7 0.3938 0.175 1>
    Cobalt=<rgba 0 0.0185 1 1>
    Crimson=<rgba 1 0 0.0271 1>
    DefaultColor=<rgba 0.5917 0.45 3 1>
    Grey=<rgba 0.3 0.3 0.3 1>
    Lime=<rgba 0.5 1 0 1>
    Orange=<rgba 1 0.3488 0 1>
    Pink=<rgba 2 0 1.2727 1>
    Purple=<rgba 0.8333 0 1 1>
    R1 Black:LinearColor=<rgba 0.2 0.2 0.2 1>
    R1 BurntSienna:LinearColor=<rgba 1.4 0.6825 0.35 1>
    R1 Cobalt:LinearColor=<rgba 0.5 0.5277 2 1>
    R1 Crimson:LinearColor=<rgba 1 0 0.0271 1>
    R1 DefaultColor=<rgba 2.5 1.9688 0.375 0.5>
    R1 ForestGreen:LinearColor=<rgba 0 1.5 0.0758 1>
    R1 Gold:LinearColor=<rgba 5 5 5 1>
    R1 Grey:LinearColor=<rgba 0.5 0.5 0.5 1>
    R1 Lime:LinearColor=<rgba 1 2 0 1>
    R1 Orange:LinearColor=<rgba 2 0.6976 0 1>
    R1 Pink:LinearColor=<rgba 2 0 1.2727 1>
    R1 Purple:LinearColor=<rgba 1.0833 0 1.3 1>
    R1 Saffron:LinearColor=<rgba 2 1.5 0 1>
    R1 SkyBlue:LinearColor=<rgba 0 1.3308 2 1>
    R1 TitaniumWhite:LinearColor=<rgba 2 2 2 1>
    R2 Black:LinearColor=<rgba 0.1 0.1 0.1 1>
    R2 BurntSienna:LinearColor=<rgba 0.0642 0.0198 0.0079 1>
    R2 Crimson:LinearColor=<rgba 0.6 0 0.0163 1>
    R2 DefaultColor:LinearColor=<rgba 0 0 0 1>
    R2 Orange:LinearColor=<rgba 1 0.3488 0 1>
    R2 Pink:LinearColor=<rgba 0.75 0.0898 0.5099 1>
    R2 Purple:LinearColor=<rgba 0.3694 0 0.9 1>
    R2 Saffron:LinearColor=<rgba 0.3 0.23 0 1>
    R2 TitaniumWhite:LinearColor=<rgba 0.35 0.35 0.35 1>
    R3 DefaultColor:LinearColor=<rgba 0 0 0 1>
    R3 Pink:LinearColor=<rgba 0 0 0 1>
    R4 Black:LinearColor=<rgba 0.13 0.13 0.13 1>
    R4 BurntSienna:LinearColor=<rgba 1.3 0.7107 0.416 1>
    R4 Cobalt:LinearColor=<rgba 0 0.0462 2.5 1>
    R4 Crimson:LinearColor=<rgba 2 0.3 0.3461 1>
    R4 DefaultColor:LinearColor=<rgba 20 15 0 1>
    R4 ForestGreen:LinearColor=<rgba 0 1.5 0.0758 1>
    R4 Grey:LinearColor=<rgba 0.5656 0.6 0.5656 1>
    R4 Lime:LinearColor=<rgba 1 2 0 1>
    R4 Orange:LinearColor=<rgba 5 0.0083 0 1>
    R4 Pink:LinearColor=<rgba 2.6 0 1.3 1>
    R4 Purple:LinearColor=<rgba 2 0 1.3333 1>
    R4 Saffron:LinearColor=<rgba 2 1.8333 0 1>
    R4 SkyBlue:LinearColor=<rgba 0 1.3308 2 1>
    R4 TitaniumWhite:LinearColor=<rgba 0.8765 0.9 0.8765 1>
    Saffron=<rgba 2 1.8333 0 1>
    SkyBlue=<rgba 0 1.3308 2 1>
    TitaniumWhite=<rgba 0.8 0.8 0.8 1>
    Use R1 Painted Overrides=true
    Use R2 Painted Overrides=true
    Use R3 Painted Overrides=true
    Use R4 Painted Overrides=true
   ) {module=VC_PaintedVariantsOverride001}
   /SetVariables(Particles.Visivily_tag:NiagaraInt32=&SpawnGroup()) {isSetParameters=true}
   &InitializeParticle(
    Color Mode=::NewEnumerator0
    Sprite Rotation Mode=::NewEnumerator2
    Sprite Size Mode=::NewEnumerator3
    Uniform Sprite Size=/VehicleCosmetics/FX/Niagara/Modules/MergeEmitters/VC_FloatBySpawnGroup(
     SpawnGroup or ExecIndex:NiagaraInt32=0
     CustomInt:NiagaraInt32=0
     MaximumSpawnGroups:NiagaraInt32=5
     Float SG 0:NiagaraFloat=85
     Float SG 1:NiagaraFloat=65
     Float SG 2:NiagaraFloat=85
     Float SG 3:NiagaraFloat=35
     Float SG 4:NiagaraFloat=45
     Float SG 5:NiagaraFloat=0
    )
    UsePositionOffset=false
    Write Material Random=true
   )
   &PartitionParticles(Number Of Partitions=$Emitter.PortCount)
   &SkeletalMeshLocation(
    Bone / Socket Index=$Particles.PartitionParticles.Partition
    Bone / Socket Sampling Mode=::NewEnumerator5
    MaskSpawnGroup=false
    Mesh Sampling Type=::NewEnumerator6
    Sampled Position Offset=&AddVectorToPosition(Position=<-50 0 0>, Vector=$User.DTCarOffset)
    Skeletal Mesh=$System.SkeletalMesh
   )
   &AddVelocity(Velocity=<-1500 0 0>) {enabled=false}
  ]
  particleUpdateScript [
   &ParticleState()
   &InheritSourceMovement(
    Applied Position Scale=&Multiply_VectorByFloat(Vector=<1 1 1>, Float=1)
    ApplyForce=false
    ApplyPosition=true
    ApplyPositionFalloff=false
   ) {enabled=false}
   &SpriteRotationRate(Rotation Rate=-10)
   &ScaleSpriteSize(Uniform Curve Sprite Scale=DataInterface() {propertyValues {Curve {keys [<at 0 0 RCIM_Cubic>, {tangentMode=RCTM_User, time=1, value=2}]}}})
   &SolveForcesAndVelocity(Clamp Velocity=false, Limit Acceleration=false)
   /SetVariables(Particles.DynamicCoreBlack:Vector4f=&MakeVector4(
     X=&FloatFromCurve(FloatCurve=DataInterface() {propertyValues {Curve {keys [<at 0 8>,<at 1 8>]}}})
     Y=&FloatFromCurve(FloatCurve=DataInterface() {propertyValues {Curve {keys [<at 0 6>,<at 1 6>]}}})
     Z=0
     W=&FloatFromCurve(FloatCurve=DataInterface() {propertyValues {Curve {keys [<at 0 0.03>,<at 1 0.05>]}}})
    )) {isSetParameters=true}
   &DynamicMaterialParameters(
    Index 0 Param 1=&VC_FloatBySpawnGroup(
     SpawnGroup or ExecIndex:NiagaraInt32=0
     CustomInt:NiagaraInt32=0
     MaximumSpawnGroups:NiagaraInt32=5
     Float SG 0:NiagaraFloat=&FloatFromCurve(FloatCurve=DataInterface() {propertyValues {Curve {keys [<at 0 0>,<at 1 0.25>]}}})
     Float SG 1:NiagaraFloat=&FloatFromCurve(FloatCurve=DataInterface() {propertyValues {Curve {keys [<at 0 0>,<at 1 0.25>]}}})
     Float SG 2:NiagaraFloat=&FloatFromCurve(FloatCurve=DataInterface() {propertyValues {Curve {keys [<at 0 0>,<at 1 0.25>]}}})
     Float SG 3:NiagaraFloat=&FloatFromCurve(FloatCurve=DataInterface() {propertyValues {Curve {keys [<at 0 8>,<at 1 8>]}}})
     Float SG 4:NiagaraFloat=&FloatFromCurve(FloatCurve=DataInterface() {propertyValues {Curve {keys [<at 0 8>,<at 1 8>]}}})
     Float SG 5:NiagaraFloat=0
    )
    Index 0 Param 2=&VC_FloatBySpawnGroup(
     SpawnGroup or ExecIndex:NiagaraInt32=0
     CustomInt:NiagaraInt32=0
     MaximumSpawnGroups:NiagaraInt32=5
     Float SG 0:NiagaraFloat=&FloatFromCurve(FloatCurve=DataInterface() {propertyValues {Curve {keys [<at 0.7 8>,<at 1 16>]}}})
     Float SG 1:NiagaraFloat=&FloatFromCurve(FloatCurve=DataInterface() {propertyValues {Curve {keys [<at 0 6>,<at 1 6>]}}})
     Float SG 2:NiagaraFloat=&FloatFromCurve(FloatCurve=DataInterface() {propertyValues {Curve {keys [<at 0.7 8>,<at 1 16>]}}})
     Float SG 3:NiagaraFloat=&FloatFromCurve(FloatCurve=DataInterface() {propertyValues {Curve {keys [<at 0 6>,<at 1 6>]}}})
     Float SG 4:NiagaraFloat=&FloatFromCurve(FloatCurve=DataInterface() {propertyValues {Curve {keys [<at 0 8>,<at 1 8>]}}})
     Float SG 5:NiagaraFloat=0
    )
    Index 0 Param 3=&VC_FloatBySpawnGroup(
     SpawnGroup or ExecIndex:NiagaraInt32=0
     CustomInt:NiagaraInt32=0
     MaximumSpawnGroups:NiagaraInt32=5
     Float SG 0:NiagaraFloat=0
     Float SG 1:NiagaraFloat=0
     Float SG 2:NiagaraFloat=1
     Float SG 3:NiagaraFloat=0
     Float SG 4:NiagaraFloat=0
     Float SG 5:NiagaraFloat=0
    )
    Index 0 Param 4=&VC_FloatBySpawnGroup(
     SpawnGroup or ExecIndex:NiagaraInt32=0
     CustomInt:NiagaraInt32=0
     MaximumSpawnGroups:NiagaraInt32=5
     Float SG 0:NiagaraFloat=&FloatFromCurve(FloatCurve=DataInterface() {propertyValues {Curve {keys [<at 0 0.5>,<at 1 0.3>]}}})
     Float SG 1:NiagaraFloat=&FloatFromCurve(FloatCurve=DataInterface() {propertyValues {Curve {keys [<at 0 1.2>,<at 1 0.1>]}}})
     Float SG 2:NiagaraFloat=&FloatFromCurve(FloatCurve=DataInterface() {propertyValues {Curve {keys [<at 0 0.3>,<at 1 0.3>]}}})
     Float SG 3:NiagaraFloat=&FloatFromCurve(FloatCurve=DataInterface() {propertyValues {Curve {keys [<at 0 0.03>,<at 1 0.3>]}}})
     Float SG 4:NiagaraFloat=&FloatFromCurve(FloatCurve=DataInterface() {propertyValues {Curve {keys [<at 0 0>,<at 1 0>]}}})
     Float SG 5:NiagaraFloat=&FloatFromCurve(FloatCurve=DataInterface() {propertyValues {Curve {keys [<at 0 1>,<at 1 0>]}}})
    )
    Module.Index0Float0Write=true
    Module.Index0Float1Write=true
    Module.Index0Float2Write=true
    Module.Index0Float3Write=true
   )
   &DynamicMaterialParameters(
    Index 0 Param 1=&FloatFromCurve(FloatCurve=DataInterface() {propertyValues {Curve {keys [<at 0 0>,<at 1 0.1>]}}})
    Index 0 Param 2=&FloatFromCurve(FloatCurve=DataInterface() {propertyValues {Curve {keys [<at 0.7 8>,<at 1 16>]}}})
    Index 0 Param 3=&FloatFromCurve(FloatCurve=DataInterface() {propertyValues {Curve {keys [<at 0 0>,<at 1 0>]}}})
    Index 0 Param 4=&FloatFromCurve(FloatCurve=DataInterface() {propertyValues {Curve {keys [<at 0 0.45>,<at 1 0.2>]}}})
    Module.Index0Float0Write=true
    Module.Index0Float1Write=true
    Module.Index0Float2Write=true
    Module.Index0Float3Write=true
   ) {enabled=false, module=DynamicMaterialParameters001}
  ]
  rendererClasses [&NiagaraSpriteRendererProperties]
  renderers [
   {rendererClass=&NiagaraSpriteRendererProperties}
   {rendererClass=&NiagaraSpriteRendererProperties, rendererIndex=1}
   {rendererClass=&NiagaraSpriteRendererProperties, rendererIndex=2}
   {rendererClass=&NiagaraSpriteRendererProperties, rendererIndex=3}
  ]
 }
 Lightning {
  emitterUpdateScript [
   &EmitterState(Inactive Response=::NewEnumerator1, Life Cycle Mode=::NewEnumerator1, UseLoopDelay=false)
   &VC_FilteredSocketCount()
   &SpawnRate(SpawnGroup=1, SpawnRate=8, Use Spawn Probability=false)
  ]
  particleSpawnScript [
   &VC_PaintedVariantsOverride(
    Black=<rgba 0.5 0.5 0.5 1>
    BurntSienna=<rgba 2 0.6175 0.2465 1>
    Cobalt=<rgba 0.9 0.9018 1 1>
    Crimson=<rgba 1 0.5 0.5136 1>
    DefaultColor=<rgba 1 0.75 0 1>
    ForestGreen=<rgba 1 2 1.0505 1>
    Grey=<rgba 1 1 1 1>
    Lime=<rgba 1.7 2 1.4 1>
    Orange=<rgba 1 0.8046 0.7 1>
    Pink=<rgba 2 0.8 1.5636 1>
    Purple=<rgba 2.0521 0 5 1>
    SkyBlue=<rgba 1.5 3.8289 5 1>
    TitaniumWhite=<rgba 25 25 25 1>
   )
   &InitializeParticle(
    Color Mode=::NewEnumerator0
    Lifetime Max=0.15
    Lifetime Min=0.1
    Lifetime Mode=::NewEnumerator1
    Mesh Scale Mode=::NewEnumerator3
    Mesh Uniform Scale=&RandomRangeFloat(Minimum=8, Maximum=12)
    Sprite Rotation Mode=::NewEnumerator2
    UsePositionOffset=false
    Write Material Random=true
   )
   &PartitionParticles(Number Of Partitions=$Emitter.PortCount)
   &SkeletalMeshLocation(
    Bone / Socket Index=$Particles.PartitionParticles.Partition
    Bone / Socket Sampling Mode=::NewEnumerator5
    MaskSpawnGroup=false
    Mesh Sampling Type=::NewEnumerator6
    Sampled Position Offset=&AddVectorToPosition(Position=<-50 0 0>, Vector=$User.DTCarOffset)
    Skeletal Mesh=$System.SkeletalMesh
   )
   &AddVelocity(Velocity=<-1500 0 0>) {enabled=false}
   &ShapeLocation(MaskSpawnGroup=false, Override Seed=false, Rotation Axis=<0 0 0>, Rotation Mode=::NewEnumerator1, Sphere Radius=1, UseOwnerScale=true)
  ]
  particleUpdateScript [
   &ParticleState()
   &UpdateMeshOrientation(Rotation Rate=-360, Rotation Vector=<360 360 0>)
   &ScaleMeshSize(Scale Factor=&RandomRangeVector(Minimum=<0.1 0.1 0.1>, Maximum=<0.125 0.1 0.1>))
   &SolveForcesAndVelocity(Clamp Velocity=false, Limit Acceleration=false)
   &ScaleColor(Scale RGB=<20 0 0>, ScaleA=true, ScaleRGB=true) {enabled=false}
   &DynamicMaterialParameters(
    Index 0 Param 1=&FloatFromCurve(FloatCurve=DataInterface() {propertyValues {Curve {keys [<at 0 3 RCIM_Cubic>, {arriveTangent=-3.6197, tangentMode=RCTM_User, time=1, value=0}]}}})
    Index 0 Param 4=&FloatFromCurve(FloatCurve=DataInterface() {propertyValues {Curve {keys [<at 0 0>,<at 1 0>]}}})
    Module.Index0Float0Write=true
    Module.Index0Float1Write=true
    Module.Index0Float2Write=true
    Module.Index0Float3Write=true
   )
  ]
  rendererClasses [&NiagaraMeshRendererProperties, &NiagaraSpriteRendererProperties]
  renderers [
   {rendererClass=&NiagaraMeshRendererProperties, rendererIndex=0}
   {rendererClass=&NiagaraSpriteRendererProperties, rendererIndex=1}
  ]
 }
 RibbonLightninGroup {
  bEnabled=false
  emitterUpdateScript [
   &EmitterState()
   &VC_FilteredSocketCount()
   &VC_RateQualityHandler()
   &VC_SpawnRibbon_v1(Number of Ribbons=$Emitter.PortCount, SpawnRate=20)
  ]
  particleSpawnScript [
   &VC_PaintedVariantsOverride() {enabled=false}
   &VC_PaintedVariantsOverride(
    DefaultColor=<rgba 5 2 0.75 1>
    R1 Crimson:LinearColor=<rgba 1 0 0.0271 1>
    R1 Gold:LinearColor=<rgba 5 5 5 1>
    R2 Black:LinearColor=<rgba 0.1 0.1 0.1 1>
    R2 DefaultColor:LinearColor=<rgba 1 0.4 0.15 1>
    R3 DefaultColor:LinearColor=<rgba 5 2 0.75 1>
    Use R1 Painted Overrides=true
    Use R2 Painted Overrides=true
    Use R3 Painted Overrides=true
    Use R4 Painted Overrides=true
   ) {module=VC_PaintedVariantsOverride001}
   &VC_InitializeRibbon_v1(Color=$Particles.Color, Enable Skeletal Position=true, Ribbon Width=15, Write Material Random=false)
   &AddVelocity(Velocity=<-1000 0 0>) {module=AddVelocity001}
  ]
  particleUpdateScript [
   &ParticleState()
   &SolveForcesAndVelocity(Clamp Velocity=false, Limit Acceleration=false)
   &VC_UpdateRibbon_v1(Enable Skeletal Position=true)
   &ScaleColor(
    Scale Alpha=&FloatFromCurve(FloatCurve=DataInterface() {propertyValues {Curve {keys [<at 0 0 RCIM_Cubic>,<at 0.1 0.5 RCIM_Cubic>,<at 1 0>]}}})
    Scale RGB=&VectorFromFloat(Value=&FloatFromCurve(FloatCurve=DataInterface() {propertyValues {Curve {keys [<at 0 3 RCIM_Cubic>,<at 1 1>]}}}))
    ScaleA=true
    ScaleRGB=true
   )
   /SetVariables(
    Particles.Linear Color=&VC_DynamicScaleColor(Color to Scale:LinearColor=<rgba 1 1 1 1>, RGB Scale:Vector3f=<0 0 0>, Alpha Scale:NiagaraFloat=0)
    Particles.SecondTrailColor:LinearColor=&VC_DynamicScaleColor(
     Color to Scale:LinearColor=$Particles.R2 Color
     RGB Scale:Vector3f=<55 1 1>
     Alpha Scale:NiagaraFloat=&FloatFromCurve(FloatCurve=DataInterface() {propertyValues {Curve {keys [{tangentMode=RCTM_User, time=0, value=1}, {tangentMode=RCTM_User, time=1, value=0}]}}})
    )
   ) {isSetParameters=true}
   &DynamicMaterialParameters(
    Index 0 Param 1=&VC_FloatBySpawnGroup(
     SpawnGroup or ExecIndex:NiagaraInt32=0
     CustomInt:NiagaraInt32=0
     MaximumSpawnGroups:NiagaraInt32=5
     Float SG 0:NiagaraFloat=&RandomRangeFloat(Minimum=0, Maximum=1)
     Float SG 1:NiagaraFloat=0
     Float SG 2:NiagaraFloat=&RandomRangeFloat(Minimum=1, Maximum=0)
     Float SG 3:NiagaraFloat=0
     Float SG 4:NiagaraFloat=0
     Float SG 5:NiagaraFloat=0
    )
    Index 0 Param 2=&VC_FloatBySpawnGroup(
     SpawnGroup or ExecIndex:NiagaraInt32=0
     CustomInt:NiagaraInt32=0
     MaximumSpawnGroups:NiagaraInt32=5
     Float SG 0:NiagaraFloat=&RandomRangeFloat(Minimum=0, Maximum=1)
     Float SG 1:NiagaraFloat=0
     Float SG 2:NiagaraFloat=&RandomRangeFloat(Minimum=1, Maximum=0)
     Float SG 3:NiagaraFloat=0
     Float SG 4:NiagaraFloat=0
     Float SG 5:NiagaraFloat=0
    )
    Index 0 Param 3=&VC_FloatBySpawnGroup(
     SpawnGroup or ExecIndex:NiagaraInt32=0
     CustomInt:NiagaraInt32=0
     MaximumSpawnGroups:NiagaraInt32=5
     Float SG 0:NiagaraFloat=0
     Float SG 1:NiagaraFloat=0
     Float SG 2:NiagaraFloat=0
     Float SG 3:NiagaraFloat=0
     Float SG 4:NiagaraFloat=0
     Float SG 5:NiagaraFloat=0
    )
    Index 0 Param 4=&VC_FloatBySpawnGroup(
     SpawnGroup or ExecIndex:NiagaraInt32=0
     CustomInt:NiagaraInt32=0
     MaximumSpawnGroups:NiagaraInt32=5
     Float SG 0:NiagaraFloat=1
     Float SG 1:NiagaraFloat=0
     Float SG 2:NiagaraFloat=0.5
     Float SG 3:NiagaraFloat=0
     Float SG 4:NiagaraFloat=0
     Float SG 5:NiagaraFloat=0
    )
    Module.Index0Float0Write=true
    Module.Index0Float1Write=true
    Module.Index0Float2Write=true
    Module.Index0Float3Write=true
   )
  ]
  rendererClasses [&NiagaraRibbonRendererProperties]
  renderers [
   {rendererClass=&NiagaraRibbonRendererProperties}
   {rendererClass=&NiagaraRibbonRendererProperties, rendererIndex=1}
   {rendererClass=&NiagaraRibbonRendererProperties, rendererIndex=2}
   {rendererClass=&NiagaraRibbonRendererProperties, rendererIndex=3}
   {rendererClass=&NiagaraRibbonRendererProperties, rendererIndex=4}
  ]
 }
 RibbonLightningGroup {
  emitterUpdateScript [
   &EmitterState(Inactive Response=::NewEnumerator1, Life Cycle Mode=::NewEnumerator1, Loop Duration=0, UseLoopDelay=false)
   &VC_FilteredSocketCount()
   &VC_SpawnRibbon_v1(Number of Ribbons=$Emitter.PortCount, Spawn Time=1, SpawnRate=60)
  ]
  particleSpawnScript [
   &VC_PaintedVariantsOverride() {enabled=false}
   /SetVariables(Particles.Ribbon_0=65, Particles.Ribbon_01=50, Particles.Ribbon_02=40, Particles.Ribbon_03=50, Particles.Ribbon_04=65) {isSetParameters=true}
   &VC_PaintedVariantsOverride(
    DefaultColor=<rgba 0 0 0 1>
    R1 Black:LinearColor=<rgba 0 0 0 1>
    R1 BurntSienna:LinearColor=<rgba 0 0 0 1>
    R1 Cobalt:LinearColor=<rgba 0 0 0 1>
    R1 Crimson:LinearColor=<rgba 0 0 0 1>
    R1 DefaultColor=<rgba 0 0 0 1>
    R1 ForestGreen:LinearColor=<rgba 0 0 0 1>
    R1 Gold:LinearColor=<rgba 5 5 5 1>
    R1 Grey:LinearColor=<rgba 1.2 1.4 1.3 1>
    R1 Lime:LinearColor=<rgba 0 0 0 1>
    R1 Orange:LinearColor=<rgba 0 0 0 1>
    R1 Pink:LinearColor=<rgba 0 0 0 1>
    R1 Purple:LinearColor=<rgba 0 0 0 1>
    R1 Saffron:LinearColor=<rgba 0 0 0 1>
    R1 SkyBlue:LinearColor=<rgba 0 0 0 1>
    R1 TitaniumWhite:LinearColor=<rgba 4.5 5 4.5 1>
    R2 Black:LinearColor=<rgba 0 0 0 1>
    R2 BurntSienna:LinearColor=<rgba 1 0.2116 0 1>
    R2 Cobalt:LinearColor=<rgba 0 1.0175 55 1>
    R2 Crimson:LinearColor=<rgba 1.5 0 0.0407 1>
    R2 DefaultColor:LinearColor=<rgba 25 7.3529 0 1>
    R2 ForestGreen:LinearColor=<rgba 0 5 0.2526 1>
    R2 Grey:LinearColor=<rgba 0.1 0.1 0.1 1>
    R2 Lime:LinearColor=<rgba 2.5 5 0 1>
    R2 Orange:LinearColor=<rgba 20 6.9761 0 1>
    R2 Pink:LinearColor=<rgba 55 6.5886 37.3958 1>
    R2 Purple:LinearColor=<rgba 22.5728 0 55 1>
    R2 Saffron:LinearColor=<rgba 5 3.8333 0 1>
    R2 SkyBlue:LinearColor=<rgba 0 1.9962 3 1>
    R3 Black:LinearColor=<rgba 0 0 0 1>
    R3 BurntSienna:LinearColor=<rgba 0.15 0.0375 0 1>
    R3 Cobalt:LinearColor=<rgba 0 0.0055 0.3 1>
    R3 Crimson:LinearColor=<rgba 0.5 0 0.0136 1>
    R3 DefaultColor:LinearColor=<rgba 0.0556 0 1 3>
    R3 Grey:LinearColor=<rgba 0 0 0 1>
    R3 Lime:LinearColor=<rgba 0.125 0.5 0 1>
    R3 Orange:LinearColor=<rgba 1 0.05 0 1>
    R3 Pink:LinearColor=<rgba 0.5 0 0.3182 1>
    R3 Purple:LinearColor=<rgba 0.0566 0 0.6795 1>
    R3 SkyBlue:LinearColor=<rgba 0 1.3308 2 1>
    R3 TitaniumWhite:LinearColor=<rgba 0 0 0 1>
    R4 BurntSienna:LinearColor=<rgba 0 0 0 1>
    R4 Cobalt:LinearColor=<rgba 0 0 0 1>
    R4 Crimson:LinearColor=<rgba 0 0 0 1>
    R4 DefaultColor:LinearColor=<rgba 0 0 0 1>
    R4 ForestGreen:LinearColor=<rgba 0 0 0 1>
    R4 Grey:LinearColor=<rgba 1.7857 2 1.7857 1>
    R4 Lime:LinearColor=<rgba 0 0 0 1>
    R4 Orange:LinearColor=<rgba 0 0 0 1>
    R4 Pink:LinearColor=<rgba 0 0 0 1>
    R4 Purple:LinearColor=<rgba 0 0 0 1>
    R4 Saffron:LinearColor=<rgba 0 0 0 1>
    R4 SkyBlue:LinearColor=<rgba 0 0 0 1>
    R4 TitaniumWhite:LinearColor=<rgba 7.9603 8 7.92 1>
    R5 Black:LinearColor=<rgba 0.35 0.35 0.35 1>
    R5 BurntSienna:LinearColor=<rgba 2 0.6667 0 1>
    R5 Cobalt:LinearColor=<rgba 0.4 0.4296 2 1>
    R5 Crimson:LinearColor=<rgba 3 0.3 0 1>
    R5 DefaultColor:LinearColor=<rgba 6 2.1882 0.6 1>
    R5 ForestGreen:LinearColor=<rgba 0 2 0.1011 1>
    R5 Grey:LinearColor=<rgba 0.4 0.4 0.4 1>
    R5 Lime:LinearColor=<rgba 1 2 0 1>
    R5 Orange:LinearColor=<rgba 5 0.8333 0 1>
    R5 Pink:LinearColor=<rgba 2 0.2396 1.3598 1>
    R5 Purple:LinearColor=<rgba 1.026 0 2.5 1>
    R5 Saffron:LinearColor=<rgba 2 1.5333 0 1>
    R5 SkyBlue:LinearColor=<rgba 0 1.9962 3 1>
    R5 TitaniumWhite:LinearColor=<rgba 3 3 3 1>
    Use R1 Painted Overrides=true
    Use R2 Painted Overrides=true
    Use R3 Painted Overrides=true
    Use R4 Painted Overrides=true
    Use R5 Painted Overrides=true
   ) {module=VC_PaintedVariantsOverride001}
   &VC_InitializeRibbon_v1(
    Color=$Particles.Color
    Custom Offset=true
    Enable Skeletal Position=true
    Local Vector Offset=&AddVectorToPosition(Position=<-15 0 0>, Vector=$User.DTCarOffset)
    Offset Amount=1
    Ribbon Width=0
   )
   &AddVelocity(Velocity=<-3000 0 0>) {module=AddVelocity001}
  ]
  particleUpdateScript [
   &ParticleState()
   &InheritSourceMovement(
    Applied Position Scale=&VectorFromFloat(Value=&FloatFromCurve(FloatCurve=DataInterface() {propertyValues {Curve {keys [
          {interpMode=RCIM_Cubic, tangentMode=RCTM_Break, time=0, value=1}
          {arriveTangent=-2.9025, interpMode=RCIM_Cubic, tangentMode=RCTM_Break, time=1, value=0}
         ]}}}))
    ApplyForce=false
    ApplyPosition=true
    ApplyPositionFalloff=false
   )
   &SolveForcesAndVelocity(Clamp Velocity=false, Limit Acceleration=false)
   &ScaleColor(
    Scale Alpha=&FloatFromCurve(FloatCurve=DataInterface() {propertyValues {Curve {keys [<at 0 0 RCIM_Cubic>,<at 0.1 0.5 RCIM_Cubic>,<at 1 0>]}}})
    Scale RGB=&VectorFromFloat(Value=&FloatFromCurve(FloatCurve=DataInterface() {propertyValues {Curve {keys [<at 0 3 RCIM_Cubic>,<at 1 1>]}}}))
    ScaleA=true
    ScaleRGB=true
   ) {enabled=false}
   &VC_UpdateRibbon_v1(Enable Skeletal Position=true)
   /SetVariables(
    Particles.Ribbon_Rate1:LinearColor=&VC_DynamicScaleColor(
     Color to Scale:LinearColor=$Particles.R1 Color
     RGB Scale:Vector3f=&VectorFromFloat(Value=&FloatFromCurve(FloatCurve=DataInterface() {propertyValues {Curve {keys [<at 0 0.01 RCIM_Cubic>,<at 1 0.01>]}}}))
     Alpha Scale:NiagaraFloat=&FloatFromCurve(FloatCurve=DataInterface() {propertyValues {Curve {keys [<at 0 0 RCIM_Cubic>,<at 0.1 1.5 RCIM_Cubic>]}}})
    )
    Particles.Ribbon_Rate2:LinearColor=&VC_DynamicScaleColor(
     Color to Scale:LinearColor=$Particles.R2 Color
     RGB Scale:Vector3f=&VectorFromFloat(Value=&FloatFromCurve(FloatCurve=DataInterface() {propertyValues {Curve {keys [<at 0 2 RCIM_Cubic>,<at 1 0.5>]}}}))
     Alpha Scale:NiagaraFloat=&FloatFromCurve(FloatCurve=DataInterface() {propertyValues {Curve {keys [<at 0 0 RCIM_Cubic>,<at 0.1 1 RCIM_Cubic>,<at 1 0>]}}})
    )
    Particles.Ribbon_Rate3:LinearColor=&VC_DynamicScaleColor(
     Color to Scale:LinearColor=$Particles.R3 Color
     RGB Scale:Vector3f=&VectorFromFloat(Value=&FloatFromCurve(FloatCurve=DataInterface() {propertyValues {Curve {keys [<at 0 60 RCIM_Cubic>,<at 1 1>]}}}))
     Alpha Scale:NiagaraFloat=&FloatFromCurve(FloatCurve=DataInterface() {propertyValues {Curve {keys [<at 0 0 RCIM_Cubic>,<at 0.1 1 RCIM_Cubic>,<at 1 0>]}}})
    )
    Particles.Ribbon_Rate4:LinearColor=&VC_DynamicScaleColor(
     Color to Scale:LinearColor=$Particles.R4 Color
     RGB Scale:Vector3f=&VectorFromFloat(Value=&FloatFromCurve(FloatCurve=DataInterface() {propertyValues {Curve {keys [<at 0 0.002 RCIM_Cubic>,<at 1 0.002>]}}}))
     Alpha Scale:NiagaraFloat=&FloatFromCurve(FloatCurve=DataInterface() {propertyValues {Curve {keys [<at 0 0 RCIM_Cubic>, {arriveTangent=1, interpMode=RCIM_Cubic, leaveTangent=1, time=0.1, value=1}, <at 1 1>]}}})
    )
   ) {isSetParameters=true}
   /SetVariables(
    Particles.Dynamic Black Lightning:Vector4f=&MakeVector4(
     X=&RandomRangeFloat(Minimum=0, Maximum=0)
     Y=&RandomRangeFloat(Minimum=0.5, Maximum=1)
     Z=0
     W=-1
    )
    Particles.Dynamic Core Lightning:Vector4f=&MakeVector4(
     X=&RandomRangeFloat(Minimum=0.1, Maximum=0)
     Y=&RandomRangeFloat(Minimum=0, Maximum=0)
     Z=0
     W=0.5
    )
    Particles.Dynamic InvertedLightning:Vector4f=&MakeVector4(
     X=&RandomRangeFloat(Minimum=-1, Maximum=1.3)
     Y=&RandomRangeFloat(Minimum=1, Maximum=0.3)
     Z=0
     W=1
    )
    Particles.Dynamic Lightning Inverted:Vector4f=&MakeVector4(
     X=&RandomRangeFloat(Minimum=0, Maximum=0)
     Y=&RandomRangeFloat(Minimum=0.5, Maximum=1)
     Z=0
     W=1
    )
    Particles.dynamic LightningBlack:Vector4f=&MakeVector4(
     X=&RandomRangeFloat(Minimum=-1, Maximum=1.3)
     Y=&RandomRangeFloat(Minimum=1, Maximum=0.3)
     Z=0
     W=-1
    )
   ) {isSetParameters=true}
   /SetVariables() {isSetParameters=true}
  ]
  rendererClasses [&NiagaraRibbonRendererProperties]
  renderers [
   {rendererClass=&NiagaraRibbonRendererProperties}
   {rendererClass=&NiagaraRibbonRendererProperties, rendererIndex=1}
   {rendererClass=&NiagaraRibbonRendererProperties, rendererIndex=2}
   {rendererClass=&NiagaraRibbonRendererProperties, rendererIndex=3}
   {rendererClass=&NiagaraRibbonRendererProperties, rendererIndex=4}
   {rendererClass=&NiagaraRibbonRendererProperties, rendererIndex=5}
  ]
 }
}
summary {
 systemName=NS_Boost_Trace_MP
 emitters [
  {emitterName=FlareGroup, rendererClasses [&NiagaraSpriteRendererProperties]}
  {emitterName=Lightning, rendererClasses [&NiagaraMeshRendererProperties, &NiagaraSpriteRendererProperties]}
  {bEnabled=false, emitterName=RibbonLightninGroup, rendererClasses [&NiagaraRibbonRendererProperties]}
  {emitterName=RibbonLightningGroup, rendererClasses [&NiagaraRibbonRendererProperties]}
 ]
 userVariables [
  User.bIsFrontendPreview {type=NiagaraBool}
  User.bIsPainted {type=NiagaraBool}
  User.ColorVariantID {type=NiagaraInt32}
  User.DTCarOffset {type=Vector3f, description="", defaultValue {struct=Vector3f, value=<0 0 0>}}
  User.PaintedColor {type=LinearColor, description="", defaultValue {struct=LinearColor, value=<rgba 1 1 1 1>}}
  User.PaintedScalar {type=NiagaraFloat, defaultValue=1, description=""}
  User.PortCount {type=NiagaraInt32, defaultValue=1}
 ]
}
