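# Translucent unlit lightning material for Niagara ribbons and sprites in the Boost Trace vehicle cosmetic.
# An animated flipbook samples RGB lightning, with dynamic parameters selecting channels, adding sharpness and blending toward an inverted-masked opacity.
# UV gradients fade edges, particle color tints emissive through MF_FX_EmissiveAdjustments, and the bolt also scales a subtle refraction term.
# A second flipbook, gradient stripe and noise-distorted band mask form side branches with no visible consumer.
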Material /VehicleCosmetics/Boosts/Boost_Trace/Material/MT_Boost_Trace_DarkLightning
graph {
 properties {
  bAutomaticallySetUsageInEditor=false
  bUsedWithNiagaraRibbons=true
  bUsedWithNiagaraSprites=true
  blendMode=BLEND_Translucent
  refractionDepthBias=1
  refractionMethod=RM_2DOffset
  shadingModel=MSM_Unlit
  usedShadingModels=Unlit
 }
 nodes [
  n1 = DynamicParameter() {paramNames [TextureSwitch, TextureSwitch, Sharpness, ToggleInverted]}
  n2 = Multiply(A=LinearInterpolate(A=@n1.TextureSwitch), B=@n1.Sharpness)
  n3 = Panner(Coordinate=TextureCoordinate() {uTiling=4}, Time=Constant() {r=1})
  n4 = ScalarParameter() {defaultValue=4, group=FlipBook, parameterName=02_Columns}
  n5 = ScalarParameter() {defaultValue=4, group=FlipBook, parameterName=02_Rows}
  n6 = AppendVector(A=@n4, B=@n5)
  n7 = Multiply(A=@n4, B=@n5)
  n8 = Frac(None=Multiply(A=Time(), B=ScalarParameter() {defaultValue=1, group=FlipBook, parameterName=02_Speed}))
  n9 = Floor(None=Multiply(A=AppendVector(A=@n7, B=@n5), B=AppendVector(A=@n8, B=@n8)))
  n10 = TextureSampleParameter2D(UVs=Add(A=Divide(A=@n3, B=@n6), B=Divide(A=@n9, B=@n6))) {parameterName=Texture_02, texture=/VehicleCosmetics/Boosts/Boost_Trace/Textures/T_VCVFX_Boost_Trace_LightningRGB}
  Add(A=@n2, B=LinearInterpolate(B=@n10.B, Alpha=@n1.TextureSwitch))
  n11 = ComponentMask(None=TextureCoordinate()) {r=true}
  n12 = Multiply(A=@n11)
  n13 = TextureCoordinate()
  n14 = Multiply(A=Time(), B=Constant() {r=0.5})
  n15 = ComponentMask(None=TextureSample(UVs=Panner(Coordinate=@n13, Time=@n14) {speedX=-1}) {texture=/DealershipSharedLibrary/FX/Textures/Noise/T_VCVFX_Flow03_N}.RGB) {g=true, r=true}
  n16 = Constant() {r=0.1}
  n17 = TextureSample(UVs=Add(A=TextureCoordinate() {uTiling=2}, B=Multiply(A=@n15, B=@n16))) {texture=/DealershipSharedLibrary/FX/Textures/Gradient_Masks/T_VCVFX_GradientStripe_A}
  n18 = LinearInterpolate(A=@n10.R, B=Multiply(A=Multiply(A=@n10.G, B=@n11), B=Constant() {r=15}), Alpha=@n1.TextureSwitch)
  n19 = Constant() {r=8}
  n20 = TextureCoordinate()
  n21 = Multiply(A=Add(A=@n2, B=Multiply(A=@n18, B=@n19)), B=OneMinus(None=ComponentMask(None=@n20) {r=true}))
  Add(A=Multiply(A=Multiply(A=@n12, B=@n17.RGB), B=Constant() {r=8}), B=@n21)
  MaterialExpressionVertexColor_0 = ParticleColor()
  n22 = Desaturation(None=@MaterialExpressionVertexColor_0.RGB, Fraction=ScalarParameter() {parameterName=Saturation})
  MaterialFunctionCall(Emissive=Multiply(A=ScalarParameter() {defaultValue=8, parameterName=Brightness}, B=Multiply(A=@n22, B=@n21))) {materialFunction=&MF_FX_EmissiveAdjustments}
  n23 = TextureCoordinate()
  n24 = ComponentMask(None=@n23) {r=true}
  MaterialExpressionConstantClamp_6 = Saturate(None=ConstantBiasScale(None=OneMinus(None=@n24)) {bias=-0.005, scale=2})
  MaterialExpressionConstantClamp_7 = Saturate(None=ConstantBiasScale(None=@n24) {bias=-0.005, scale=6})
  n25 = ComponentMask(None=@n23) {g=true}
  n26 = Multiply(A=@n25, B=OneMinus(None=@n25))
  n27 = Constant() {r=5}
  MaterialExpressionConstantClamp_5 = Saturate(None=Multiply(A=Multiply(A=@MaterialExpressionConstantClamp_6, B=@MaterialExpressionConstantClamp_7), B=Multiply(A=@n26, B=@n27)))
  n28 = TextureCoordinate()
  n29 = Add(A=@n28, B=Constant2Vector() {g=-0.2})
  n30 = Constant() {r=0.9}
  n31 = Panner(Coordinate=TextureCoordinate() {vTiling=0.125}, Time=Constant() {r=0.5}) {speedY=0.5}
  n32 = TextureSample(UVs=Panner(Coordinate=TextureCoordinate() {uTiling=0.25, vTiling=0.25}) {speedX=-0.125, speedY=0.01}) {texture=&T_VCVFX_Flow03_N}
  n33 = Multiply(A=Constant() {r=0.05}, B=AppendVector(A=@n32.R, B=@n32.R))
  n34 = TextureSample(UVs=Panner(Coordinate=Add(A=@n31, B=@n33)) {speedX=-1}) {texture=/VehicleCosmetics/Boosts/Boost_Trace/Textures/T_VCVFX_Boost_Trace_DistortNoise}
  n35 = Panner(Coordinate=TextureCoordinate() {uTiling=1.125, vTiling=-0.125}) {speedY=-0.5}
  n36 = TextureSample(UVs=Panner(Coordinate=Add(A=@n35)) {speedX=-1}) {texture=&T_VCVFX_Boost_Trace_DistortNoise}
  n37 = Multiply(A=Constant() {r=0.155}, B=ComponentMask(None=Add(A=@n34.RGB, B=@n36.RGB)) {g=true, r=true})
  n38 = ComponentMask(None=Add(A=Rotator(Coordinate=@n29, Time=@n30) {speed=3.14}, B=@n37)) {g=true}
  MaterialExpressionConstantClamp_1 = Saturate(None=ConstantBiasScale(None=@n38) {bias=-0.015, scale=1})
  MaterialExpressionConstantClamp_2 = Saturate(None=ConstantBiasScale(None=OneMinus(None=@n38)) {bias=-0.01, scale=1})
  n39 = Multiply(A=@MaterialExpressionConstantClamp_1, B=@MaterialExpressionConstantClamp_2)
  n40 = Add(A=@n28, B=Constant2Vector() {g=0.2})
  n41 = Constant() {r=0.12}
  n42 = ComponentMask(None=Add(A=@n37, B=Rotator(Coordinate=@n40, Time=@n41) {speed=3.14})) {g=true}
  MaterialExpressionConstantClamp_3 = Saturate(None=ConstantBiasScale(None=@n42) {bias=-0.015, scale=1})
  MaterialExpressionConstantClamp_4 = Saturate(None=ConstantBiasScale(None=OneMinus(None=@n42)) {bias=-0.01, scale=1})
  n43 = Multiply(A=@MaterialExpressionConstantClamp_3, B=@MaterialExpressionConstantClamp_4)
  MaterialExpressionConstantClamp_0 = Saturate(None=Multiply(A=Constant() {r=150}, B=Multiply(A=@n39, B=@n43)))
  Multiply(A=@MaterialExpressionConstantClamp_5, B=@MaterialExpressionConstantClamp_0)
  n44 = Saturate(None=Multiply(A=OneMinus(None=@n11), B=Constant() {r=8}))
  n45 = Constant() {r=0.25}
  Multiply(A=Multiply(A=@n44, B=Multiply(A=@n21, B=@n45)), B=ScalarParameter() {defaultValue=0.01, parameterName=Power_Refraction}) {constB=0.2}
  n46 = Saturate(None=OneMinus(None=@n21))
  n47 = Multiply(A=LinearInterpolate(A=@n21, B=Multiply(A=@n21, B=@n46), Alpha=@n1.ToggleInverted), B=@MaterialExpressionConstantClamp_5)
  Saturate(None=Multiply(A=@n44, B=Multiply(A=@MaterialExpressionVertexColor_0.A, B=@n47)))
  n48 = Panner(Coordinate=TextureCoordinate() {uTiling=4}, Time=Constant() {r=1}) {speedX=-1, speedY=1}
  n49 = ScalarParameter() {defaultValue=4, group=FlipBook, parameterName=01_Columns}
  n50 = ScalarParameter() {defaultValue=4, group=FlipBook, parameterName=01_Rows}
  n51 = AppendVector(A=@n49, B=@n50)
  n52 = Multiply(A=@n49, B=@n50)
  n53 = Frac(None=Multiply(A=Time(), B=ScalarParameter() {defaultValue=1, group=FlipBook, parameterName=01_Speed}))
  n54 = Floor(None=Multiply(A=AppendVector(A=@n52, B=@n50), B=AppendVector(A=@n53, B=@n53)))
  TextureSampleParameter2D(UVs=Add(A=Divide(A=@n48, B=@n51), B=Divide(A=@n54, B=@n51))) {parameterName=Texture_01, texture=/VehicleCosmetics/Boosts/Boost_Trace/Textures/T_VCVFX_Boost_Trace_Lightning}
 ]
}
