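# Translucent unlit effect material for the Trace vehicle boost, named for an explosion expanding-ball lighting effect and flagged for Niagara mesh particles and skeletal meshes.
# Particle color is slightly desaturated and fed into the MF_FX_EmissiveAdjustments function.
# A separate chain offsets particle color and multiplies it by a V-coordinate wipe driven by the Swipe dynamic parameter.
# No link from that chain to an output appears in the graph.
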Material /VehicleCosmetics/Boosts/Boost_Trace/Material/MT_Boost_Trace_ExplosionExpBallLighting
graph {
 properties {
  bAutomaticallySetUsageInEditor=false
  bUsedWithMeshDeformer=true
  bUsedWithNiagaraMeshParticles=true
  bUsedWithSkeletalMesh=true
  blendMode=BLEND_Translucent
  shadingModel=MSM_Unlit
  usedShadingModels=Unlit
 }
 nodes [
  MaterialExpressionVertexColor_3 = ParticleColor()
  MaterialExpressionConstantClamp_14 = Saturate(None=Subtract(A=@MaterialExpressionVertexColor_3.RGB, B=Constant() {r=0.2}))
  n1 = TextureCoordinate()
  MaterialExpressionMeshEmitterDynamicParameter_0 = DynamicParameter() {paramNames [Swipe,Param2,Param3,Param4]}
  MaterialExpressionConstantClamp_13 = Saturate(None=Subtract(A=ComponentMask(None=@n1) {g=true}, B=OneMinus(None=@MaterialExpressionMeshEmitterDynamicParameter_0.Swipe)))
  MaterialExpressionConstantClamp_12 = Saturate(None=Multiply(A=Add(B=@MaterialExpressionConstantClamp_14), B=@MaterialExpressionConstantClamp_13))
  MaterialExpressionMeshEmitterVertexColor_0 = ParticleColor()
  MaterialFunctionCall(Emissive=Desaturation(None=@MaterialExpressionMeshEmitterVertexColor_0.RGB, Fraction=Constant() {r=0.15})) {materialFunction=&MF_FX_EmissiveAdjustments}
 ]
}
