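# Unlit translucent Niagara sprite material that draws radial rays around a soft core, named for the Boost Trace wallbreak effect.
# Noise sampled by polar angle, offset by dynamic speed, forms rays and distorts the radial falloff.
# An adjustable inner circle and a small color-weighted center dot add to the opacity, which passes through depth fade.
# Squared particle color feeds the emissive adjustment function.
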Material /VehicleCosmetics/Boosts/Boost_Trace/Material/MT_Boost_Trace_WallbreakRays
graph {
 properties {
  bAutomaticallySetUsageInEditor=false
  bUsedWithNiagaraSprites=true
  blendMode=BLEND_Translucent
  shadingModel=MSM_Unlit
  usedShadingModels=Unlit
 }
 nodes [
  MaterialExpressionVertexColor_2 = ParticleColor()
  MaterialFunctionCall(Emissive=Multiply(A=@MaterialExpressionVertexColor_2.RGB, B=@MaterialExpressionVertexColor_2.RGB)) {materialFunction=&MF_FX_EmissiveAdjustments}
  n1 = MaterialFunctionCall() {materialFunction=&VectorToRadialValue}
  n2 = Constant2Vector() {r=1}
  n3 = Constant()
  n4 = DynamicParameter() {paramNames [speed, power, bias, CenterDistort]}
  n5 = TextureSample(UVs=Add(A=Multiply(A=@n1.Vector Converted to Angle, B=@n2), B=AppendVector(A=@n3, B=@n4.speed))) {samplerType=SAMPLERTYPE_LinearColor, texture=/DealershipSharedLibrary/FX/Textures/Noise/T_VCVFX_NoiseClouds}
  n6 = Add(A=@n5.G, B=Constant() {r=-0.5})
  n7 = Add(A=Add(A=TextureCoordinate(), B=Multiply(A=@n6, B=@n4.CenterDistort)), B=Constant() {r=-0.5})
  n8 = SquareRoot(None=DotProduct(A=@n7, B=@n7))
  MaterialExpressionConstantClamp_10 = Saturate(None=Power(Base=OneMinus(None=@n8), Exp=Constant() {r=16}))
  n9 = ScalarParameter() {defaultValue=2.5, parameterName=InnerCircleSize}
  n10 = Add(
   A=Multiply(A=Power(Base=@MaterialExpressionConstantClamp_10, Exp=@n9), B=ScalarParameter() {defaultValue=1, parameterName=InnerCircleHardness})
   B=Multiply(A=@MaterialExpressionConstantClamp_10, B=Power(Base=@n5.G, Exp=@n4.power))
  )
  n11 = Add(A=Panner(Coordinate=TextureCoordinate() {uTiling=3, vTiling=3}, Time=Constant() {r=1}) {speedX=1.5, speedY=1.5}, B=Constant() {r=-3})
  n12 = DotProduct(A=@n11, B=@n11)
  MaterialExpressionConstantClamp_11 = Saturate(None=ConstantBiasScale(None=OneMinus(None=@n12)) {bias=-0.99, scale=60})
  MaterialExpressionConstantClamp_9 = Saturate(None=Multiply(A=@MaterialExpressionVertexColor_2.RGB, B=Multiply(A=@MaterialExpressionConstantClamp_11, B=@MaterialExpressionConstantClamp_11)))
  MaterialExpressionConstantClamp_8 = Saturate(None=Multiply(A=Add(A=@n10, B=@MaterialExpressionConstantClamp_9), B=ScalarParameter() {defaultValue=1, parameterName=Opacity}))
  MaterialFunctionCall(Opacity In=@MaterialExpressionConstantClamp_8, DepthFade Distance=@n4.bias) {materialFunction=&MF_VCVFX_DepthFade}
 ]
}
